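# Animation Blueprint for the M_MED_GreatPool male body cosmetic, adding secondary motion to the pose copied from its attached parent mesh.
# A Control Rig, bone corrections and pelvis-space cloth RigidBody feed a cached pose, then spine-space RigidBody variants blend on skydiving.
# The event graph updates LOD and gravity override, and maps crouched left-thigh yaw to a jacket-bone offset.
# Crouch transitions reset dynamics; several other jacket bone nodes are disconnected from the output.
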AnimBlueprint /BRCosmetics/Characters/Player/Male/Medium/Bodies/M_MED_GreatPool/Meshes/M_MED_GreatPool_AnimBP
target_skeleton=&/M_MED_GreatPool_Skeleton
graphs [
 AnimGraph {nodes [
   n1 = Get thigh_rot_l()
   "Transform (Modify) Bone - Bone: dyn_jacket_fr_l_mstr"(Alpha=@n1, Translation="0.123,-0.941,-0.328", Rotation="32.135,-4.173,49.923", Scale="1,1,1") {
    rotationMode=BMM_Additive
    scaleMode=BMM_Additive
    scaleSpace=BCS_BoneSpace
    translationMode=BMM_Additive
    translationSpace=BCS_BoneSpace
    alphaBoolBlend {blendOption=Linear}
    alphaScaleBiasClamp {bInterpResult=true, interpSpeedDecreasing=9, interpSpeedIncreasing=30}
    boneToModify {boneName=dyn_jacket_fr_l_mstr}
    rotation=<rot 32.135 -4.173 49.923>
    translation=<0.123 -0.941 -0.328>
   }
   n2 = Get thigh_rot_l()
   "Transform (Modify) Bone - Bone: dyn_jacket_fr_l_mstr"(Alpha=@n2, Translation="0,-1,1", Rotation="20,0,0", Scale="1,1,1") {
    rotationMode=BMM_Additive
    scaleMode=BMM_Additive
    scaleSpace=BCS_BoneSpace
    translationMode=BMM_Additive
    translationSpace=BCS_BoneSpace
    alphaBoolBlend {blendOption=Linear}
    alphaScaleBiasClamp {bInterpResult=true, interpSpeedDecreasing=9, interpSpeedIncreasing=30}
    boneToModify {boneName=dyn_jacket_fr_l_mstr}
    rotation {pitch=20}
    translation {y=-1, z=1}
   }
   "Transform (Modify) Bone - Bone: dyn_jacket_s_l_mstr"(Alpha=@n1, Translation="-1.688,-1.330,-1.631", Rotation="-17.920,-3.749,31.829", Scale="0.7,1,1") {
    rotationMode=BMM_Additive
    scaleMode=BMM_Additive
    scaleSpace=BCS_BoneSpace
    translationMode=BMM_Additive
    translationSpace=BCS_BoneSpace
    alphaBoolBlend {blendOption=Linear}
    alphaScaleBiasClamp {bInterpResult=true, interpSpeedDecreasing=7, interpSpeedIncreasing=7}
    boneToModify {boneName=dyn_jacket_s_l_mstr}
    rotation=<rot -17.92 -3.749 31.829>
    scale {x=0.7}
    translation=<-1.688 -1.33 -1.631>
   }
   n3 = "Transform (Modify) Bone - Bone: dyn_jacket_fr_l_mstr"(Alpha=@n2, Translation="0,0,0", Rotation="-20,0,50", Scale="1,1,1") {
    rotationMode=BMM_Additive
    scaleMode=BMM_Additive
    scaleSpace=BCS_BoneSpace
    translationMode=BMM_Additive
    translationSpace=BCS_BoneSpace
    alphaBoolBlend {blendOption=Linear}
    alphaScaleBiasClamp {bInterpResult=true, interpSpeedDecreasing=9, interpSpeedIncreasing=30}
    boneToModify {boneName=dyn_jacket_fr_l_mstr}
    rotation {pitch=-20, roll=50}
   }
   "Transform (Modify) Bone - Bone: dyn_jacket_s_l_mstr"(ComponentPose=@n3, Alpha=@n2, Translation="0,0,0", Rotation="-20,0,30", Scale="0.6,1,1") {
    rotationMode=BMM_Additive
    scaleMode=BMM_Additive
    scaleSpace=BCS_BoneSpace
    translationMode=BMM_Additive
    translationSpace=BCS_BoneSpace
    alphaBoolBlend {blendOption=Linear}
    alphaScaleBiasClamp {bInterpResult=true, interpSpeedDecreasing=7, interpSpeedIncreasing=7}
    boneToModify {boneName=dyn_jacket_s_l_mstr}
    rotation {pitch=-20, roll=30}
    scale {x=0.6}
   }
   n4 = "Transform (Modify) Bone - Bone: dyn_jacket_fr_l_mstr"(Alpha=@n1, Translation="0,0,0", Rotation="0,0,0", Scale="0.8,1,1") {
    rotationMode=BMM_Additive
    scaleMode=BMM_Additive
    scaleSpace=BCS_BoneSpace
    translationMode=BMM_Additive
    translationSpace=BCS_BoneSpace
    alphaBoolBlend {blendOption=Linear}
    alphaScaleBiasClamp {bInterpResult=true, interpSpeedDecreasing=9, interpSpeedIncreasing=30}
    boneToModify {boneName=dyn_jacket_fr_l_mstr}
    scale {x=0.8}
   }
   n5 = "Transform (Modify) Bone - Bone: dyn_jacket_fr_l_02"(ComponentPose=@n4, Alpha=@n1, Translation="0,0,0", Rotation="0,0,0", Scale="1,1,1") {
    rotationMode=BMM_Additive
    scaleMode=BMM_Additive
    scaleSpace=BCS_BoneSpace
    translationMode=BMM_Additive
    translationSpace=BCS_BoneSpace
    alphaBoolBlend {blendOption=Linear}
    alphaScaleBiasClamp {bInterpResult=true, interpSpeedDecreasing=9, interpSpeedIncreasing=30}
    boneToModify {boneName=dyn_jacket_fr_l_02}
   }
   "Transform (Modify) Bone - Bone: dyn_jacket_s_l_mstr"(ComponentPose=@n5, Alpha=@n1, Translation="0,0,0", Rotation="0,0,0", Scale="1,1,1") {
    rotationMode=BMM_Additive
    scaleMode=BMM_Additive
    scaleSpace=BCS_BoneSpace
    translationMode=BMM_Additive
    translationSpace=BCS_BoneSpace
    alphaBoolBlend {blendOption=Linear}
    alphaScaleBiasClamp {bInterpResult=true, interpSpeedDecreasing=7, interpSpeedIncreasing=7}
    boneToModify {boneName=dyn_jacket_s_l_mstr}
   }
   n6 = Get GravityOverride()
   n7 = RigidBody!(
    ComponentPose=Local To Component(LocalPose=Use cached pose 'Pre sim'())
    bAlphaBoolEnabled:bool
    OverrideWorldGravity:Vector=@n6
    ExternalForce
   ) -> (Pose) {
    alphaInputType=Bool
    bClampLinearTranslationLimitToRefPose=true
    bOverrideWorldGravity=true
    overridePhysicsAsset=&/M_MED_GreatPool_Physics
    simulationSpace=BaseBoneSpace
    alphaBoolBlend {blendOption=Linear}
    baseBoneRef {boneName=spine_03}
    componentAppliedLinearAccClamp=<2000 2000 2000>
    componentLinearAccScale=<0.05 0.05 0.05>
    componentLinearVelScale=<0.05 0.05 0.05>
    overrideWorldGravity {z=-980}
    simSpaceSettings {
     maxAngularAcceleration=300
     maxAngularVelocity=3
     maxLinearAcceleration=500
     maxLinearVelocity=300
     velocityScaleZ=0
     worldAlpha=1
     externalLinearDragV=<1 1>
    }
   }
   n8 = RigidBody(ComponentPose=Local To Component(LocalPose=Use cached pose 'Pre sim'()), OverrideWorldGravity=@n6) {
    alphaInputType=Bool
    bClampLinearTranslationLimitToRefPose=true
    bOverrideWorldGravity=true
    overridePhysicsAsset=&M_MED_GreatPool_Physics
    simulationSpace=BaseBoneSpace
    alphaBoolBlend {blendOption=Linear}
    baseBoneRef {boneName=spine_03}
    componentAppliedLinearAccClamp=<1000 1000 1000>
    componentLinearAccScale=<0.05 0.05 0.05>
    componentLinearVelScale=<0.05 0.05 0.05>
    overrideWorldGravity {z=-980}
    simSpaceSettings {
     maxAngularAcceleration=300
     maxAngularVelocity=3
     maxLinearAcceleration=500
     maxLinearVelocity=300
     velocityScaleZ=0
     worldAlpha=1
     externalLinearDragV=<1 1>
    }
   }
   n9 = Blend Poses by bool(BlendPose_0=Component To Local(ComponentPose=@n8), BlendPose_1=Component To Local(ComponentPose=@n7), bActiveValue=Get bIsSkyDiving()) {blendType=Linear}
   Output Pose(Result=Component To Local(ComponentPose=Local To Component(LocalPose=@n9))) {layerGroup=None}
   n10 = Copy Pose From Mesh() {bUseAttachedParent=true}
   n11 = Control Rig!(
    Source=Component To Local(ComponentPose=Local To Component(LocalPose=@n10))
    Alpha
    is RBAN ON:bool=false
    is LOD 2:bool=Get Is Lod 2() -> (Is Lod 2:bool)
   ) -> (Pose) {
    alphaBoolBlend {blendOption=Linear}
    controlRigAssetReference {blueprintRigClass=&/M_MED_GreatPool_CtrlRig.M_MED_GreatPool_CtrlRig_C}
   }
   n12 = "Transform (Modify) Bone - Bone: shirt_spine_05_correction"(ComponentPose=Local To Component(LocalPose=@n11), Alpha=1, Translation="0,0,0", Rotation="0,0,0", Scale="0.1,0.1,0.1") {
    rotationMode=BMM_Additive
    scaleMode=BMM_Additive
    scaleSpace=BCS_BoneSpace
    translationMode=BMM_Additive
    translationSpace=BCS_BoneSpace
    alphaBoolBlend {blendOption=Linear}
    boneToModify {boneName=shirt_spine_05_correction}
    scale=<0.1 0.1 0.1>
   }
   n13 = "Transform (Modify) Bone - Bone: shirt_spine_04_correction"(ComponentPose=@n12, Alpha=1, Translation="-0.542,4.754,-0.041", Rotation="0,0,0", Scale="0.1,0.1,0.1") {
    rotationMode=BMM_Additive
    scaleMode=BMM_Additive
    scaleSpace=BCS_BoneSpace
    translationMode=BMM_Additive
    translationSpace=BCS_BoneSpace
    alphaBoolBlend {blendOption=Linear}
    boneToModify {boneName=shirt_spine_04_correction}
    scale=<0.1 0.1 0.1>
    translation=<-0.542 4.754 -0.041>
   }
   n14 = "Constraint - star_l (P)"(ComponentPose=@n13, Alpha=0.5, ConstraintWeights_0=1) {
    alphaBoolBlend {blendOption=Linear}
    boneToModify {boneName=star_l}
    constraintSetup [{targetBone {boneName=thigh_l}}]
    constraintWeights [1]
   }
   n15 = "Transform (Modify) Bone - Bone: star_l"(ComponentPose=@n14, Alpha=1, Translation="0.135,-1.000,0.000", Rotation="0,0,0", Scale="1,1,1") {translationMode=BMM_Additive, alphaBoolBlend {blendOption=Linear}, boneToModify {boneName=star_l}, translation=<0.135 -1>}
   n16 = "Constraint - star_r (P)"(ComponentPose=@n15, Alpha=0.5, ConstraintWeights_0=1) {
    alphaBoolBlend {blendOption=Linear}
    boneToModify {boneName=star_r}
    constraintSetup [{targetBone {boneName=thigh_r}}]
    constraintWeights [1]
   }
   n17 = "Transform (Modify) Bone - Bone: star_r"(ComponentPose=@n16, Alpha=1, Translation="-0.248,-1.000,-0.000", Rotation="0,0,0", Scale="1,1,1") {translationMode=BMM_Additive, alphaBoolBlend {blendOption=Linear}, boneToModify {boneName=star_r}, translation=<-0.2478 -1 0>}
   n18 = RigidBody!(ComponentPose=@n17, Alpha:float, OverrideWorldGravity="0,0,-980", ExternalForce) -> (Pose) {
    bOverrideWorldGravity=true
    overridePhysicsAsset=&/M_MED_GreatPool_Cloth_Physics
    simulationSpace=BaseBoneSpace
    alphaBoolBlend {blendOption=Linear}
    baseBoneRef {boneName=pelvis}
    componentAppliedLinearAccClamp=<1000 1000 1000>
    componentLinearAccScale=<0.3 0.3 0.3>
    componentLinearVelScale=<0.3 0.3 0.3>
    overrideWorldGravity {z=-980}
    simSpaceSettings {
     maxAngularAcceleration=20
     maxAngularVelocity=1
     maxLinearAcceleration=20
     maxLinearVelocity=50
     velocityScaleZ=0.2
     worldAlpha=1
     externalLinearDragV=<1 1 1>
    }
   }
   n19 = "Transform (Modify) Bone - Bone: dyn_jacket_fr_r_mstr"(ComponentPose=@n18, Alpha=@n1, Translation="0,0,0", Rotation="32.421,-34.395,2.916", Scale="1,1,1") {
    rotationMode=BMM_Additive
    scaleMode=BMM_Additive
    scaleSpace=BCS_BoneSpace
    translationMode=BMM_Additive
    translationSpace=BCS_BoneSpace
    alphaBoolBlend {blendOption=Linear}
    alphaScaleBiasClamp {bInterpResult=true, interpSpeedDecreasing=9, interpSpeedIncreasing=30}
    boneToModify {boneName=dyn_jacket_fr_r_mstr}
    rotation=<rot 32.421 -34.395 2.916>
   }
   Save cached pose 'Pre sim'(Pose=Component To Local(ComponentPose=@n19)) {cacheName=Pre sim}
   n20 = "Transform (Modify) Bone - Bone: dyn_jacket_s_l_mstr"(Alpha=@n2, Translation="0,0,0", Rotation="25,0,0", Scale="1,1,1") {
    rotationMode=BMM_Additive
    scaleMode=BMM_Additive
    scaleSpace=BCS_BoneSpace
    translationMode=BMM_Additive
    translationSpace=BCS_BoneSpace
    alphaBoolBlend {blendOption=Linear}
    alphaScaleBiasClamp {bInterpResult=true, interpSpeedDecreasing=7, interpSpeedIncreasing=7}
    boneToModify {boneName=dyn_jacket_s_l_mstr}
    rotation {pitch=25}
   }
   Reroute Node(InputPin:Component Space Pose Link Structure=@n20) -> (OutputPin:Component Space Pose Link Structure)
  ]}
 EventGraph {nodes [
   Event BlueprintUpdateAnimation()
   > Set Is Lod 2(Is Lod 2:bool="integer >= integer"(A=Get CurrentLOD(), B=2)) -> (Output_Get:bool)
   > n1 = Sequence!(execute) -> (then_0, then_1, then_2:Exec)
   n2 = Branch(execute=@n1.then_2, Condition=Get bIsCrouching())
   n3 = GetDeltaTransformFromRefPose!(self=GetOwnerSkeletalMeshComponent(), BoneName=thigh_l, BaseName=pelvis) -> (
    ReturnValue_Location:Vector
    ReturnValue_Rotation_Roll:float
    ReturnValue_Rotation_Pitch:float
    ReturnValue_Rotation_Yaw:float
    ReturnValue_Scale:Vector
   )
   > Set thigh_rot_l(thigh_rot_l=MapRangeClamped(Value=@n3.ReturnValue_Rotation_Yaw, InRangeA=-100, InRangeB=-20, OutRangeA=1, OutRangeB=0))
   Set thigh_rot_l(execute=@n2.else, thigh_rot_l=0)
   n4 = Gravity Override State Function DEPRECATED(
    execute=@n1.then_0
    CharacterPartAnimInstance=Self-Reference()
    GameTimeInSeconds=GetGameTimeInSeconds()
    GravityOverrideParams=Get Gravity Override Params()
   )
   > Set GravityOverride(GravityOverride=@n4)
   n5 = Event OnBeginCrouching()
   ResetDynamics(execute=[@n5.then, @n6.then], InTeleportType=ResetPhysics)
   > ResetDynamics(InTeleportType=TeleportPhysics)
   n6 = Event OnEndCrouching()
   MapRangeClamped(Value=@n3.ReturnValue_Rotation_Yaw, InRangeA=-200, InRangeB=-100)
   NOT Boolean(A=false)
   n7 = OR Boolean!(A=Get bIsPlayingEmote(), B=Get bIsInFrontEnd(), C:bool=Get bIsDBNO()) -> (ReturnValue)
   Set JacketControl(JacketControl:bool="Equal (Float)"(A=SelectFloat(A=0, B=1, bPickA=@n7), B=1)) -> (Output_Get:bool)
  ]}
]
variables [
 Control Rig Alpha {type=boolean}
 ControlRigAlpha {type=boolean}
 Gravity Override Params {
  type=GravityOverrideParamsStruct
  default {skydive_FloatingDown {gravityOverride=<0 -200 100>, windOffset=<0 -20 100>}, skydive_GoingUp {gravityOverride=<0 -300 200>}}
 }
 GravityOverride {type=Vector}
 Is Lod 2 {type=boolean}
 JacketControl {type=boolean}
 Thigh R Crouch {type=number}
 thigh_rot_l {type=number}
 thigh_rot_r {type=number}
]
